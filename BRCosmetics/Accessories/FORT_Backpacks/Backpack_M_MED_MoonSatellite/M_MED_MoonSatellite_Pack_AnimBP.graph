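# Animation Blueprint for the MoonSatellite backpack cosmetic, with logic for spinning its fish bone.
# Each update, the EventGraph advances the FishRotation variable at a constant pitch rate using delta time, through a Create Constant Rotation function.
# The AnimGraph passes the input pose through a Modify Bone node that replaces dyn_fish_center's bone-space rotation.
# That node's rotation input shows zero, with no visible link to FishRotation.
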AnimBlueprint /BRCosmetics/Accessories/FORT_Backpacks/Backpack_M_MED_MoonSatellite/M_MED_MoonSatellite_Pack_AnimBP
target_skeleton=&/M_MED_MoonSatellite_Pack_Skeleton
graphs [
 AnimGraph {nodes [
   n1 = "Transform (Modify) Bone - Bone: dyn_fish_center"(ComponentPose=Local To Component(LocalPose=Input Pose()), Alpha=1, Translation="0,0,0", Rotation="0,0,0", Scale="1,1,1") {rotationMode=BMM_Replace, rotationSpace=BCS_BoneSpace, alphaBoolBlend {blendOption=Linear}, boneToModify {boneName=dyn_fish_center}}
   Output Pose(Result=Component To Local(ComponentPose=@n1)) {layerGroup=None}
  ]}
 EventGraph {nodes [
   n1 = Event BlueprintUpdateAnimation()
   > n2 = Create Constant Rotation Function!(
    execute
    RotationRate:Rotator="-22,0,0"
    RotationVar:"Rotator (by ref)"=Get FishRotation() -> (FishRotation:Rotator)
    DeltaTime:double=@n1.DeltaTimeX
   ) -> (then, RotationOut:Rotator)
   > Set FishRotation(FishRotation:Rotator=@n2) -> (Output_Get:Rotator)
   TryGetPawnOwner()
  ]}
]
variables [FishRotation {type=Rotator}]
